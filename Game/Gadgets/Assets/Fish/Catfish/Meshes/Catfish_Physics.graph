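# Physics asset for the Catfish fish gadget mesh, defining capsule bodies and joints for its dyn_ spine and whisker bones.
# Two-bone front and back spine chains branch from a small Anim root body.
# Left and right two-segment whisker chains hang off the first front spine bone.
# Every joint has small swing and twist limit angles, with tighter twist on the whiskers.
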PhysicsAsset /Game/Gadgets/Assets/Fish/Catfish/Meshes/Catfish_Physics
bodies {
 Anim [{shapeType=Capsule, length=0.505, radius=0.505, rotation=<rot 0 -90 0>}]
 dyn_spineBack_01 [{shapeType=Capsule, length=8.1408, radius=2.7244, center=<1.5919 0 0>, rotation=<rot 90 0 0>}]
 dyn_spineBack_02 [{shapeType=Capsule, length=8.1408, radius=2.7244, center=<1.5919 0 0>, rotation=<rot 90 0 0>}]
 dyn_spineFront_01 [{shapeType=Capsule, length=8.1408, radius=2.7244, center=<1.5919 0 0>, rotation=<rot 90 0 0>}]
 dyn_spineFront_02 [{shapeType=Capsule, length=8.1408, radius=2.7244, center=<1.5919 0 0>, rotation=<rot 90 0 0>}]
 dyn_whisker01_l [{shapeType=Capsule, length=5.3765, radius=0.5, center=<1.1657 0 0>, rotation=<rot 90 0 0>}]
 dyn_whisker01_r [{shapeType=Capsule, length=5.3765, radius=0.5, center=<-1.1657 0 0>, rotation=<rot 90 0 180>}]
 dyn_whisker02_l [{shapeType=Capsule, length=5.3765, radius=0.5139, center=<1.3647 0 0>, rotation=<rot 90 0 0>}]
 dyn_whisker02_r [{shapeType=Capsule, length=5.3765, radius=0.5139, center=<-1.3647 0 0>, rotation=<rot 90 0 180>}]
}
constraints [
 {bone1Name=dyn_spineFront_01, bone2Name=Anim, swing1LimitDegrees=12, swing2LimitDegrees=12, twistLimitDegrees=12}
 {bone1Name=dyn_spineFront_02, bone2Name=dyn_spineFront_01, swing1LimitDegrees=12, swing2LimitDegrees=12, twistLimitDegrees=12}
 {bone1Name=dyn_whisker01_l, bone2Name=dyn_spineFront_01, swing1LimitDegrees=12, swing2LimitDegrees=12, twistLimitDegrees=8}
 {bone1Name=dyn_whisker02_l, bone2Name=dyn_whisker01_l, swing1LimitDegrees=12, swing2LimitDegrees=12, twistLimitDegrees=8}
 {bone1Name=dyn_spineBack_01, bone2Name=Anim, swing1LimitDegrees=12, swing2LimitDegrees=12, twistLimitDegrees=12}
 {bone1Name=dyn_spineBack_02, bone2Name=dyn_spineBack_01, swing1LimitDegrees=12, swing2LimitDegrees=12, twistLimitDegrees=12}
 {bone1Name=dyn_whisker01_r, bone2Name=dyn_spineFront_01, swing1LimitDegrees=12, swing2LimitDegrees=12, twistLimitDegrees=8}
 {bone1Name=dyn_whisker02_r, bone2Name=dyn_whisker01_r, swing1LimitDegrees=12, swing2LimitDegrees=12, twistLimitDegrees=8}
]
